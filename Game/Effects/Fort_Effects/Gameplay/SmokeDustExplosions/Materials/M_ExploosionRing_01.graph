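# Masked, unlit, two-sided material for mesh particles, apparently an animated explosion ring in smoke and dust effects.
# Two oppositely panned samples of one muzzle texture multiply, with optional stepped "staccato" timing via Ceil and particle-driven speed.
# That product blends two HDR colors on front faces, leaving back faces Color A, tinted by particle color.
# A mask combines texture, particle alpha, dynamic Erode and inverted Fresnel.
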Material /Game/Effects/Fort_Effects/Gameplay/SmokeDustExplosions/Materials/M_ExploosionRing_01
graph {
 properties {
  bUsedWithMeshParticles=true
  bUsedWithNiagaraMeshParticles=true
  blendMode=BLEND_Masked
  pixelDepthOffsetMode=PDOM_Legacy
  shadingModel=MSM_Unlit
  twoSided=true
  usedShadingModels=Unlit
 }
 nodes [
  n1 = ParticleColor()
  n2 = VectorParameter() {parameterName=Color A, defaultValue=<rgba 6 0.1 0.02 0>}
  n3 = VectorParameter() {parameterName=Color B, defaultValue=<rgba 20 3.1283 0.9 0>}
  n4 = TextureCoordinate()
  n5 = AppendVector(A=ScalarParameter() {defaultValue=2, parameterName=U Repeat A}, B=ScalarParameter() {defaultValue=1, parameterName=V Repeat A})
  n6 = Multiply(
   A=Time()
   B=Multiply(
    A=ScalarParameter() {defaultValue=12.6, parameterName=Panner A Speed}
    B=DynamicParameter() {defaultValue=<rgba 1 0 0 1>, paramNames [Erode,Param2,Param3,PannerSpeed]}.PannerSpeed
   )
  )
  n7 = Ceil(None=@n6)
  n8 = TextureSample(UVs=Panner(Coordinate=Multiply(A=@n4, B=@n5), Time=StaticSwitchParameter(True=@n7, False=@n6) {defaultValue=true, parameterName=Use Staccato Timing?}) {bFractionalPart=true, speedX=1.22}) {samplerType=SAMPLERTYPE_LinearColor, texture=/Game/Effects/Fort_Effects/Textures/Tile/T_Raygun_Muzzle_01}
  n9 = Multiply(
   A=TextureCoordinate()
   B=AppendVector(A=ScalarParameter() {defaultValue=1, parameterName=U Repeat B}, B=ScalarParameter() {defaultValue=1, parameterName=V Repeat B})
  )
  n10 = Multiply(
   A=Time()
   B=Multiply(
    A=ScalarParameter() {defaultValue=17.77, parameterName=Panner B Speed}
    B=DynamicParameter() {defaultValue=<rgba 1 0 0 1>, paramNames [Erode,Param2,Param3,PannerSpeed]}.PannerSpeed
   )
  )
  n11 = StaticSwitchParameter(True=Ceil(None=@n10), False=@n10) {defaultValue=true, parameterName=Use Staccato Timing?}
  n12 = Multiply(
   A=@n8.G
   B=TextureSample(UVs=Panner(Coordinate=@n9, Time=@n11) {bFractionalPart=true, speedX=-1.38}) {samplerType=SAMPLERTYPE_LinearColor, texture=&T_Raygun_Muzzle_01}.G
  )
  n13 = TwoSidedSign()
  n14 = Multiply(A=@n12, B=LinearInterpolate(Alpha=Clamp(None=@n13)))
  MaterialFunctionCall(Emissive=Multiply(A=@n1.RGB, B=LinearInterpolate(A=@n2.RGB, B=@n3.RGB, Alpha=@n14))) {materialFunction=&MF_FX_EmissiveAdjustments}
  n15 = ScalarParameter() {defaultValue=3, parameterName=Fresnel Exponent}
  Multiply(
   A=Multiply(A=@n1.A, B=Add(A=@n12, B=@n8.B))
   B=Add(
    A=DynamicParameter() {defaultValue=<rgba 1 0 0 1>, paramNames [Erode,Param2,Param3,PannerSpeed]}.Erode
    B=MaterialFunctionCall(Exponent=@n15) {materialFunction=&MF_Fresnel_FX}.Fresnel Inverted Unclamped
   )
  )
 ]
}
