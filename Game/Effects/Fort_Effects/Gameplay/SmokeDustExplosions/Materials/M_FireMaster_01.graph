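# Unlit translucent master material for fire effects on mesh particles, in the smoke, dust and explosion effects set.
# Fresnel and a UV V gradient, offset by the ColorErode dynamic parameter, blend three colours with an optional hard centre edge, then multiply by particle colour.
# Opacity combines Fresnel, remapped particle alpha and a V-gradient mask, with switchable depth and near-distance fades.
# An optional branch adds red vertex-colour jiggle for WPO.
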Material /Game/Effects/Fort_Effects/Gameplay/SmokeDustExplosions/Materials/M_FireMaster_01
graph {
 properties {
  bEnableResponsiveAA=true
  bUsedWithMeshParticles=true
  bUsedWithNanite=true
  bUsedWithNiagaraMeshParticles=true
  blendMode=BLEND_Translucent
  opacityMaskClipValue=0.18
  pixelDepthOffsetMode=PDOM_Legacy
  shadingModel=MSM_Unlit
  usedShadingModels=Unlit
 }
 nodes [
  n1 = VectorParameter() {parameterName=Color A1, defaultValue=<rgba 0.8 0.0887 0.024 0>}
  n2 = VectorParameter() {parameterName=Color A2, defaultValue=<rgba 1.3 0.5087 0.221 0>}
  n3 = MaterialFunctionCall(Exponent=ScalarParameter() {defaultValue=1.1, parameterName=Fresnel Exponent}) {materialFunction=&MF_Fresnel_FX}
  n4 = Power(
   Base=MaterialFunctionCall(Float2=TextureCoordinate()) {materialFunction=&BreakOutFloat2Components}.G
   Exp=ScalarParameter() {defaultValue=6, parameterName=Secondary Gradient Exponent}
  )
  n5 = DynamicParameter() {defaultValue=<rgba 0 1 1 1>, paramNames [ColorErode,Param2,Param3,Param4]}
  n6 = Saturate(None=Add(A=@n3.Fresnel Inverted, B=Subtract(A=@n4, B=@n5.ColorErode)))
  n7 = Subtract(A=@n3.Fresnel, B=@n5.ColorErode)
  n8 = ScalarParameter() {defaultValue=-0.3, parameterName=Bias LERP A}
  n9 = Clamp(None=Round(None=Add(A=@n7, B=@n8)))
  n10 = Multiply(
   A=LinearInterpolate(
    A=LinearInterpolate(A=@n1.RGB, B=@n2.RGB, Alpha=@n6)
    B=VectorParameter() {parameterName=Color B, defaultValue=<rgba 4 1.4645 0.6 0>}.RGB
    Alpha=StaticSwitchParameter(True=@n9, False=@n7) {defaultValue=true, parameterName=HardCenterEdge}
   )
   B=ParticleColor().RGB
  )
  MaterialFunctionCall(Emissive=@n10) {materialFunction=&MF_FX_EmissiveAdjustments}
  n11 = Add(A=@n3.Fresnel, B=ConstantBiasScale(None=ParticleColor().A) {bias=-0.5, scale=2})
  n12 = StaticSwitchParameter(True=DepthFade(Opacity=@n11, FadeDistance=ScalarParameter() {defaultValue=10, parameterName=Depth Fade Distance}), False=@n11) {defaultValue=true, parameterName=UseDepthFade}
  n13 = WorldPosition()
  n14 = ScalarParameter() {defaultValue=100, parameterName=Start Fade Distance}
  n15 = ScalarParameter() {defaultValue=20, parameterName=End Fade Distance}
  n16 = TextureCoordinate()
  n17 = Add(
   A=OneMinus(None=MaterialFunctionCall(Float2=@n16) {materialFunction=&BreakOutFloat2Components}.G)
   B=ScalarParameter() {defaultValue=0.2, parameterName=TranslucentMaskBias}
  )
  Multiply(
   A=StaticSwitchParameter(
    True=MaterialFunctionCall(Opacity In=@n12, Position=@n13.XYZ, Fade_Start_Distance=@n14, Fade_End_Distance=@n15) {materialFunction=&MF_Distance_SyncedFade_FX}.Result
    False=@n12
   ) {parameterName=UseNearDistanceFade}
   B=Power(Base=Saturate(None=@n17), Exp=ScalarParameter() {defaultValue=1, parameterName=TranslucentMaskExponent})
  )
  Reroute(None=@n10)
  n18 = ScalarParameter() {group=WPO, parameterName=Jiggle_Speed, sortPriority=1}
  n19 = ScalarParameter() {group=WPO, parameterName=Jiggle_Large Wave Scale, sortPriority=2}
  n20 = ScalarParameter() {group=WPO, parameterName=Jiggle_Wave Scale, sortPriority=3}
  StaticSwitchParameter(
   True=AppendVector(
    A=MaterialFunctionCall!(Speed=@n18, LargeWaveScle=@n19, WaveScale=@n20) {materialFunction=/Game/Effects/Fort_Effects/Materials/_Functions/Jiggle_WPO_VertColor_Red}.Result
    B=Constant() {r=1}
   )
   False=Constant()
  ) {group=WPO, parameterName="Use Jiggle WPO (Red Vert)"}
 ]
}
