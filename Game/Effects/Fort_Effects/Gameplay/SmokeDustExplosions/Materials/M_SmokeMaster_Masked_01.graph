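# Masked master material for mesh-particle smoke, dust and explosion effects, with stylized cel shading.
# Smoke blends smooth and stepped three-tone time-of-day colors using Fresnel on a sun/moon-camera vector, optional normal map and particle alpha.
# An optional explosion layer adds stepped hot colors with macro-UV noise, eroded by Fresnel and a dynamic parameter or debug control.
# Thickness pushes vertices along normals; opacity combines Fresnel, particle alpha and optional masks and fades.
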Material /Game/Effects/Fort_Effects/Gameplay/SmokeDustExplosions/Materials/M_SmokeMaster_Masked_01
graph {
 properties {
  bTangentSpaceNormal=false
  bUsedWithMeshParticles=true
  bUsedWithNiagaraMeshParticles=true
  blendMode=BLEND_Masked
  opacityMaskClipValue=0.45
  pixelDepthOffsetMode=PDOM_Legacy
  usedShadingModels=Default Lit
 }
 nodes [
  n1 = DynamicParameter() {defaultValue=<rgba 0 0 1 1>, paramNames [Explode Smoke LERP, Thickness, Morph Frame Random, Morph Frame Animation]}
  Add(
   A=Transform(None=VectorParameter() {group=WPO, parameterName="WPO Offset (World)"}.RGB) {transformSourceType=TRANSFORMSOURCE_Local}
   B=Multiply(A=VertexNormalWS(), B=@n1.Thickness)
  )
  n2 = VectorParameter() {group=Daytime Smoke, parameterName=Daytime Darks Smoke, defaultValue=<rgba 0.1743 0.2455 0.3206 0>}
  n3 = StaticSwitchParameter(
   True=@n2.RGB
   False=MaterialFunctionCall(
    Color_Morning=VectorParameter() {group=Morning Smoke, parameterName=Morning Darks Smoke, defaultValue=<rgba 0.2783 0.1245 0.0779 0>}.RGB
    Color_Day=@n2.RGB
    Color_Evening=VectorParameter() {group=Evening Smoke, parameterName=Evening Darks Smoke, defaultValue=<rgba 0.41 0.2237 0.1722 0>}.RGB
    Color_Night=VectorParameter() {group=Nighttime Smoke, parameterName=Nighttime Darks Smoke, defaultValue=<rgba 0.0614 0.1056 0.2117 0>}.RGB
   ) {desc=Darks, materialFunction=&Color_Time_of_Day}.Result
  ) {
   desc="If you want the same colors to be used throughout any time of day, activate this bool. Uses daytime value."
   group=Cell Shading Parameters
   parameterName=Ignore ToD Breakout?
  }
  n4 = VectorParameter() {group=Daytime Smoke, parameterName=Daytime Mids Smoke, defaultValue=<rgba 0.2041 0.2938 0.4183 0>}
  n5 = StaticSwitchParameter(
   True=@n4.RGB
   False=MaterialFunctionCall(
    Color_Morning=VectorParameter() {group=Morning Smoke, parameterName=Morning Mids Smoke, defaultValue=<rgba 0.445 0.2656 0.155 0>}.RGB
    Color_Day=@n4.RGB
    Color_Evening=VectorParameter() {group=Evening Smoke, parameterName=Evening Mids Smoke, defaultValue=<rgba 0.63 0.2681 0.1984 0>}.RGB
    Color_Night=VectorParameter() {group=Nighttime Smoke, parameterName=Nighttime Mids Smoke, defaultValue=<rgba 0.1213 0.1645 0.2939 0>}.RGB
   ) {desc=Mids, materialFunction=&Color_Time_of_Day}.Result
  ) {
   desc="If you want the same colors to be used throughout any time of day, activate this bool. Uses daytime value."
   group=Cell Shading Parameters
   parameterName=Ignore ToD Breakout?
  }
  n6 = VectorParameter() {group=Morning Smoke, parameterName=Morning Highlights Smoke, defaultValue=<rgba 0.7183 0.5222 0.3663 0>}
  n7 = VectorParameter() {group=Daytime Smoke, parameterName=Daytime Highlights Smoke, defaultValue=<rgba 0.3103 0.3962 0.6 0>}
  n8 = VectorParameter() {group=Nighttime Smoke, parameterName=Nighttime Highlights Smoke, defaultValue=<rgba 0.1955 0.4091 0.4894 0>}
  n9 = MaterialFunctionCall(
   Color_Morning=@n6.RGB
   Color_Day=@n7.RGB
   Color_Evening=VectorParameter() {group=Evening Smoke, parameterName=Evening Highlights Smoke, defaultValue=<rgba 0.64 0.4203 0.3648 0>}.RGB
   Color_Night=@n8.RGB
  ) {desc=Highlights, materialFunction=&Color_Time_of_Day}
  n10 = CollectionParameter() {
   collection=&FortniteMaterialParameters
   parameterName=SunAndMoonModelDirectionalVector
  }
  n11 = LinearInterpolate(
   A=ComponentMask(None=@n10) {b=true, g=true, r=true}
   B=CameraVectorWS()
   Alpha=ScalarParameter() {defaultValue=0.6, group=Cell Shading Parameters, parameterName=Sun/Moon Directional Shading Blend Smoke}
  )
  n12 = Multiply(
   A=AppendVector(
    A=ScalarParameter() {defaultValue=1, group=Normal Map, parameterName=Normal Map U Tile}
    B=ScalarParameter() {defaultValue=1, group=Normal Map, parameterName=Normal Map V Tile}
   )
   B=TextureCoordinate()
  )
  n13 = Time()
  n14 = ScalarParameter() {defaultValue=1, group=Normal Map, parameterName=Normal Map Panner U Speed}
  n15 = ComponentMask(None=Panner(Coordinate=@n12, Time=Multiply(A=@n13, B=@n14)) {bFractionalPart=true, speedX=0.5}) {r=true}
  n16 = ScalarParameter() {defaultValue=1, group=Normal Map, parameterName=Normal Map Panner V Speed}
  n17 = ComponentMask(None=Panner(Coordinate=@n12, Time=Multiply(A=@n13, B=@n16)) {bFractionalPart=true, speedY=0.5}) {g=true}
  n18 = TextureSampleParameter2D(UVs=StaticSwitchParameter(True=AppendVector(A=@n15, B=@n17), False=@n12) {group=Normal Map, parameterName=Pan Normal Map?}) {group=Normal Map, parameterName=Normal Map, samplerType=SAMPLERTYPE_Normal, texture=/Game/Global/Textures/Default/Noise/T_SwirlNoise_N}
  n19 = Normalize(VectorInput=Multiply(A=@n18.RGB, B=VectorParameter() {group=Normal Map, parameterName=Normal Map Intensity, defaultValue=<rgba 0.2 0.2 1 0>}.RGB))
  n20 = StaticSwitchParameter(True=Transform(None=@n19), False=PixelNormalWS()) {
   desc="If disabled, this turns off smoke shading and will output black instead."
   group=Cell Shading Parameters
   parameterName=Use Normal Map?
  }
  n21 = Add(
   A=MaterialFunctionCall(Camera Vector=Normalize(VectorInput=@n11), Vertex / Pixel Normal=@n20) {materialFunction=&MF_Fresnel_FX}.Fresnel
   B=ConstantBiasScale(None=ParticleColor().A) {bias=-1}
  )
  n22 = MaterialFunctionCall(1=@n3, 2=@n5, 3=@n9.Result, A=@n21) {materialFunction=&Lerp_Multiple_Float3}
  n23 = Multiply(A=@n21, B=ScalarParameter() {defaultValue=2, desc="Starts at 0 (so 2 = 3 steps)", group=Morning Smoke, parameterName=Cell Shading Steps})
  n24 = ScalarParameter() {defaultValue=0.25, group=Daytime Smoke, parameterName=Cell Shaded Dark/Mid Bias Day Smoke}
  n25 = StaticSwitchParameter(
   True=@n24
   False=MaterialFunctionCall(
    Color_Morning=ScalarParameter() {defaultValue=0.3, group=Morning Smoke, parameterName=Cell Shaded Dark/Mid Bias Morning Smoke}
    Color_Day=@n24
    Color_Evening=ScalarParameter() {group=Evening Smoke, parameterName=Cell Shaded Dark/Mid Bias Evening Smoke}
    Color_Night=ScalarParameter() {defaultValue=0.49, group=Nighttime Smoke, parameterName=Cell Shaded Dark/Mid Bias Night Smoke}
   ) {materialFunction=&Color_Time_of_Day}.Result
  ) {
   desc="If you want the same colors to be used throughout any time of day, activate this bool. Uses daytime value."
   group=Cell Shading Parameters
   parameterName=Ignore ToD Breakout?
  }
  n26 = Multiply(A=Round(None=Add(A=@n23, B=@n25))) {constB=2}
  n27 = ScalarParameter() {defaultValue=0.5, group=Daytime Smoke, parameterName=Cell Shaded Mid/Light Bias Day Smoke}
  n28 = ScalarParameter() {defaultValue=0.6, group=Morning Smoke, parameterName=Cell Shaded Mid/Light Bias Morning Smoke}
  n29 = ScalarParameter() {defaultValue=0.5, group=Evening Smoke, parameterName=Cell Shaded Mid/Light Bias Evening Smoke}
  n30 = ScalarParameter() {defaultValue=0.9, group=Nighttime Smoke, parameterName=Cell Shaded Mid/Light Bias Night Smoke}
  n31 = Add(
   A=@n23
   B=StaticSwitchParameter(
    True=@n27
    False=MaterialFunctionCall(Color_Morning=@n28, Color_Day=@n27, Color_Evening=@n29, Color_Night=@n30) {materialFunction=&Color_Time_of_Day}.Result
   ) {group=Cell Shading Parameters, parameterName=Ignore ToD Breakout?}
  )
  n32 = Round(None=@n31)
  n33 = LinearInterpolate(
   A=LinearInterpolate(A=@n3, B=@n5, Alpha=Clamp(None=@n26)) {desc=Dark / Mid Lerp}
   B=StaticSwitchParameter(True=@n7.RGB, False=@n9.Result) {group=Cell Shading Parameters, parameterName=Ignore ToD Breakout?}
   Alpha=Clamp(None=ConstantBiasScale(None=@n32) {bias=-1, scale=2})
  ) {desc=Mid / Light Lerp}
  n34 = ScalarParameter() {defaultValue=0.4, parameterName=Cell Shading LERP}
  n35 = VertexColor()
  n36 = Multiply(A=@n35.R, B=ScalarParameter() {defaultValue=-0.7, parameterName=AO Saturation Amount}) {constB=-2}
  n37 = Clamp(None=@n31)
  n38 = Multiply(
   A=Desaturation(None=LinearInterpolate(A=@n22.Lerp 3 Inputs, B=@n33, Alpha=@n34), Fraction=LinearInterpolate(A=@n36, Alpha=@n37) {constB=0})
   B=LinearInterpolate(A=@n35.R, Alpha=@n37)
  )
  n39 = MaterialFunctionCall() {materialFunction=&MF_Fresnel_FX}
  n40 = Time()
  n41 = ScalarParameter() {defaultValue=1, group=Explosion, parameterName=Explosion Debug LERP Speed}
  n42 = StaticSwitchParameter(True=Sine(None=Multiply(A=@n40, B=@n41)), False=ScalarParameter() {defaultValue=1, group=Explosion, parameterName=Explosion Debug LERP}) {
   defaultValue=true
   desc="If disabled, this turns off smoke shading and will output black instead."
   group=Explosion
   parameterName=Debug Explosion Use Sine?
  }
  n43 = DynamicParameter() {defaultValue=<rgba 0 0 1 1>, paramNames [Explode Smoke LERP, Thickness, Morph Frame Random, Morph Frame Animation]}
  n44 = StaticSwitchParameter(True=@n42, False=ConstantBiasScale(None=@n43.Explode Smoke LERP) {bias=-0.5, scale=2}) {desc="If disabled, this turns off smoke shading and will output black instead.", group=Explosion, parameterName=Debug Explosion LERP?}
  n45 = Clamp(None=Add(
    A=StaticSwitchParameter(True=@n39.Fresnel Inverted, False=@n39.Fresnel Inverted) {defaultValue=true, group=Explosion, parameterName=Invert Fresnel Erode}
    B=@n44
   ) {constB=0})
  n46 = StaticSwitchParameter(True=Power(Base=@n45, Exp=ScalarParameter() {defaultValue=1, group=Explosion, parameterName=Soft LERP Exponent}), False=Ceil(None=@n45)) {group=Explosion, parameterName=Use Soft LERP?}
  n47 = StaticSwitchParameter(True=Multiply(A=LinearInterpolate(A=@n38, Alpha=@n46) {constB=0}, B=ParticleColor().RGB), False=Constant()) {
   defaultValue=true
   desc="If disabled, this turns off smoke shading and will output black instead."
   group=Cell Shading Parameters
   parameterName=Use Smoke?
  }
  n48 = Multiply(A=@n47, B=ScalarParameter() {defaultValue=0.2, group=Cell Shading Parameters, parameterName=Smoke Emissive Boost}) {constB=0.5}
  n49 = VectorParameter() {group=Explosion, parameterName=Darks Explosion, defaultValue=<rgba 1.2 0.1455 0.09 0>}
  n50 = VectorParameter() {group=Explosion, parameterName=Mids Explosion, defaultValue=<rgba 3 0.334 0.285 0>}
  n51 = ScalarParameter() {defaultValue=-0.3, group=Explosion, parameterName=Cell Shaded Dark/Mid Bias Day}
  n52 = DynamicParameter() {defaultValue=<rgba 0 0 1 1>, paramNames [Explode Smoke LERP, Thickness, Morph Frame Random, Morph Frame Animation]}
  n53 = StaticSwitchParameter(True=@n42, False=@n52.Explode Smoke LERP) {desc="If disabled, this turns off smoke shading and will output black instead.", group=Explosion, parameterName=Debug Explosion LERP?}
  n54 = Clamp(None=Multiply(A=OneMinus(None=@n53), B=ScalarParameter() {defaultValue=2, group=Explosion, parameterName=Explosion Erode Color Boost}))
  n55 = Add(
   A=@n54
   B=MaterialFunctionCall(
    Brightness=ScalarParameter() {defaultValue=1.3, group=Explosion, parameterName=Shading Fresnel Brightness Explosion}
    Exponent=ScalarParameter() {defaultValue=8, group=Explosion, parameterName=Shading Fresnel Exponent Explosion}
   ) {materialFunction=&MF_Fresnel_FX}.Fresnel
  )
  n56 = StaticSwitchParameter(True=ActorPositionWS(), False=ParticlePositionWS()) {desc="If disabled, this turns off smoke shading and will output black instead.", group=Explosion, parameterName=Debug Explosion LERP?}
  n57 = Multiply(A=VertexNormalWS(), B=ScalarParameter() {defaultValue=15, parameterName=MacroUV_MeshNormalBias})
  n58 = ComponentMask(None=Multiply(A=VectorParameter() {group=Explosion, parameterName=MacroUV Panner Speed, defaultValue=<rgba 0 -1.2 0 0>}.RGB, B=Time() {period=1})) {g=true, r=true}
  n59 = Add(
   A=MaterialFunctionCall!(Plane Center=Add(A=@n56, B=@n57), Plane Normal Axis, Scale=ScalarParameter() {defaultValue=200, group=Explosion, parameterName=MacroUV Tile}) {materialFunction=/Game/Effects/Fort_Effects/Materials/_Functions/MF_MacroUV_Meshes_01}.Result
   B=@n58
  )
  n60 = Add(
   A=TextureSample(UVs=ShadingPathSwitch(Default=@n59, Mobile=@n58) {desc=Migrated from FeatureLevelSwitch}) {samplerType=SAMPLERTYPE_Normal, texture=/Game/Global/Textures/Default/Tileable_Patterns/1Normal}.R
   B=ScalarParameter() {defaultValue=0.15, group=Explosion, parameterName=MacroUV Bias}
  )
  n61 = ScalarParameter() {defaultValue=2, group=Explosion, parameterName=MacroUV Scale}
  n62 = Multiply(A=@n55, B=LinearInterpolate(A=Multiply(A=@n60, B=@n61), Alpha=@n54) {constA=1})
  n63 = Constant()
  n64 = StaticSwitchParameter(True=QualitySwitch(Default=@n62, Low=@n63, High=@n62, Medium=@n62), False=@n63) {defaultValue=true, parameterName=UseMeshMacroUV}
  n65 = Multiply(A=Clamp(None=Add(A=@n55, B=@n64)), B=ScalarParameter() {defaultValue=3, group=Explosion, parameterName=Cell Shaded Step Day})
  n66 = Clamp(None=Ceil(None=Add(A=@n51, B=@n65)))
  n67 = Ceil(None=Add(A=@n65, B=ScalarParameter() {defaultValue=-2.75, group=Explosion, parameterName=Cell Shaded Mid/Light Bias Day}))
  n68 = LinearInterpolate(
   A=@n48
   B=LinearInterpolate(
    A=LinearInterpolate(A=@n49.RGB, B=@n50.RGB, Alpha=@n66) {desc=Dark / Mid Lerp}
    B=VectorParameter() {group=Explosion, parameterName=Highlights Explosion, defaultValue=<rgba 15 8.4076 0.75 0>}.RGB
    Alpha=Clamp(None=@n67)
   ) {desc=Mid / Light Lerp}
   Alpha=@n46
  )
  n69 = StaticSwitchParameter(True=Multiply(A=@n68, B=ParticleColor().RGB), False=@n68) {
   desc="If disabled, this turns off smoke shading and will output black instead."
   group=Explosion
   parameterName=Explosion Use Particle Color?
  }
  MaterialFunctionCall(Emissive=StaticSwitchParameter(True=@n69, False=@n48) {desc="If disabled, this turns off smoke shading and will output black instead.", group=Explosion, parameterName=Use Explosion?}) {materialFunction=&MF_FX_EmissiveAdjustments}
  ScalarParameter() {group=Cell Shading Parameters, parameterName=Metallic}
  ScalarParameter() {group=Cell Shading Parameters, parameterName=Roughness}
  ScalarParameter() {group=Cell Shading Parameters, parameterName=Specular}
  n70 = ScalarParameter() {defaultValue=100, parameterName=Near Camera Fade Distance}
  n71 = TextureCoordinate()
  n72 = Power(
   Base=OneMinus(None=ComponentMask(None=@n71) {g=true})
   Exp=ScalarParameter() {defaultValue=2, group=Opacity, parameterName=UV-V Opacity Mask Power}
  )
  n73 = TextureCoordinate()
  n74 = Constant() {r=0.5}
  n75 = ScalarParameter() {defaultValue=0.5, group=Opacity, parameterName=Sphere Mask Radius}
  n76 = ScalarParameter() {defaultValue=0.2, group=Opacity, parameterName=Sphere Mask Hardness}
  n77 = MaterialFunctionCall(
   Brightness=ScalarParameter() {defaultValue=1, group=Opacity, parameterName=Opacity Fresnel Brightness}
   Exponent=ScalarParameter() {defaultValue=3, group=Opacity, parameterName=Opacity Fresnel Exponent}
   Camera Vector=@n11
  ) {materialFunction=&MF_Fresnel_FX}
  n78 = StaticSwitchParameter(True=Multiply(A=SphereMask(A=@n73, B=@n74, Radius=@n75, Hardness=@n76), B=@n77.Fresnel), False=@n77.Fresnel) {group=Opacity, parameterName=Use Sphere Mask For Opacity?}
  n79 = Add(A=StaticSwitchParameter(True=Multiply(A=@n72, B=@n78), False=@n78) {group=Opacity, parameterName=Use UV-V for Opacity?}, B=ParticleColor().A)
  n80 = StaticSwitchParameter(
   True=DepthFade(Opacity=@n79, FadeDistance=ScalarParameter() {defaultValue=20, group=Translucent Opacity, parameterName=Depth Fade Distance})
   False=@n79
  ) {group=Translucent Opacity, parameterName=Translucent Use Depth Fade?}
  StaticSwitchParameter(
   True=Multiply(A=MaterialFunctionCall(Distance=@n70) {materialFunction=&Near_Camera_Fade}.Result, B=@n80)
   False=@n80
  ) {group=Translucent Opacity, parameterName=Translucent Use Near Camera Fade}
  StaticSwitchParameter(True=@n19, False=VertexNormalWS()) {
   desc="If disabled, this turns off smoke shading and will output black instead."
   group=Cell Shading Parameters
   parameterName=Use Normal Map?
  }
 ]
}
