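# Masked, lit master material for cel-shaded smoke and explosion mesh particles.
# Stepped Fresnel, biased between sun/moon and camera vectors, picks time-of-day dark, mid and highlight smoke colors, tinted by particle color.
# An optional banded fire layer blends into emissive through inverted Fresnel and a dynamic parameter.
# Texture alpha, Fresnel plus particle alpha, and vertex alpha erode the mask, with optional normal-map and opacity panning timed by dynamic parameter.
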Material /Game/Effects/Fort_Effects/Gameplay/SmokeDustExplosions/Materials/M_SmokeMaster_Masked_Particles_01
graph {
 properties {
  bUsedWithBeamTrails=true
  bUsedWithMeshParticles=true
  bUsedWithNiagaraMeshParticles=true
  blendMode=BLEND_Masked
  opacityMaskClipValue=0.5
  pixelDepthOffsetMode=PDOM_Legacy
  translucencyLightingMode=TLM_VolumetricDirectional
  usedShadingModels=Default Lit
 }
 nodes [
  n1 = MaterialFunctionCall(
   Color_Morning=ScalarParameter() {defaultValue=1.5, group=Morning Smoke, parameterName=Morning Main Shading Fresnel Brightness Smoke}
   Color_Day=ScalarParameter() {defaultValue=1.4, group=Daytime Smoke, parameterName=Daytime Main Shading Fresnel Brightness Smoke}
   Color_Evening=ScalarParameter() {defaultValue=1.6, group=Evening Smoke, parameterName=Evening Main Shading Fresnel Brightness Smoke}
   Color_Night=ScalarParameter() {defaultValue=2, group=Nighttime Smoke, parameterName=Nighttime Main Shading Fresnel Brightness Smoke}
  ) {materialFunction=&Color_Time_of_Day}
  n2 = MaterialFunctionCall(
   Color_Morning=ScalarParameter() {defaultValue=2, group=Morning Smoke, parameterName=Morning Main Shading Fresnel Exponent Smoke}
   Color_Day=ScalarParameter() {defaultValue=4, group=Daytime Smoke, parameterName=Daytime Main Shading Fresnel Exponent Smoke}
   Color_Evening=ScalarParameter() {defaultValue=3, group=Evening Smoke, parameterName=Evening Main Shading Fresnel Exponent Smoke}
   Color_Night=ScalarParameter() {defaultValue=8, group=Nighttime Smoke, parameterName=Night Main Shading Fresnel Exponent Smoke}
  ) {materialFunction=&Color_Time_of_Day}
  n3 = CollectionParameter() {
   collection=&FortniteMaterialParameters
   parameterName=SunAndMoonModelDirectionalVector
  }
  n4 = LinearInterpolate(
   A=ComponentMask(None=@n3) {b=true, g=true, r=true}
   B=CameraVectorWS()
   Alpha=ScalarParameter() {defaultValue=0.55, group=Cell Shading Parameters, parameterName=Sun/Moon Directional Shading Blend Smoke}
  )
  n5 = ScalarParameter() {defaultValue=3, group=Morning Smoke, parameterName=Cell Shaded Step Morning Smoke}
  n6 = ScalarParameter() {defaultValue=5, group=Daytime Smoke, parameterName=Cell Shaded Step Day Smoke}
  n7 = ScalarParameter() {defaultValue=3, group=Evening Smoke, parameterName=Cell Shaded Step Evening Smoke}
  n8 = ScalarParameter() {defaultValue=3, group=Nighttime Smoke, parameterName=Cell Shaded Step Night Smoke}
  n9 = Ceil(None=Multiply(
    A=MaterialFunctionCall(Brightness=@n1.Result, Exponent=@n2.Result, Camera Vector=@n4) {materialFunction=&MF_Fresnel_FX}.Fresnel
    B=MaterialFunctionCall(Color_Morning=@n5, Color_Day=@n6, Color_Evening=@n7, Color_Night=@n8) {materialFunction=&Color_Time_of_Day}.Result
   ))
  n10 = Clamp(None=@n9) {maxDefault=2}
  n11 = MaterialFunctionCall(
   Color_Morning=ScalarParameter() {defaultValue=-0.6, group=Morning Smoke, parameterName=Cell Shaded Dark/Mid Bias Morning Smoke}
   Color_Day=ScalarParameter() {defaultValue=-1, group=Daytime Smoke, parameterName=Cell Shaded Dark/Mid Bias Day Smoke}
   Color_Evening=ScalarParameter() {defaultValue=-0.95, group=Evening Smoke, parameterName=Cell Shaded Dark/Mid Bias Evening Smoke}
   Color_Night=ScalarParameter() {defaultValue=-0.5, group=Nighttime Smoke, parameterName=Cell Shaded Dark/Mid Bias Night Smoke}
  ) {materialFunction=&Color_Time_of_Day}
  n12 = LinearInterpolate(
   A=MaterialFunctionCall(
    Color_Morning=VectorParameter() {group=Morning Smoke, parameterName=Morning Darks Smoke, defaultValue=<rgba 0.135 0.1265 0.1188 0>}.RGB
    Color_Day=VectorParameter() {group=Daytime Smoke, parameterName=Daytime Darks Smoke, defaultValue=<rgba 0.1472 0.1524 0.195 0>}.RGB
    Color_Evening=VectorParameter() {group=Evening Smoke, parameterName=Evening Darks Smoke, defaultValue=<rgba 0.29 0.1538 0.1421 0>}.RGB
    Color_Night=VectorParameter() {group=Nighttime Smoke, parameterName=Nighttime Darks Smoke, defaultValue=<rgba 0.0314 0.051 0.098 0>}.RGB
   ) {desc=Darks, materialFunction=&Color_Time_of_Day}.Result
   B=MaterialFunctionCall(
    Color_Morning=VectorParameter() {group=Morning Smoke, parameterName=Morning Mids Smoke, defaultValue=<rgba 0.445 0.3197 0.2425 0>}.RGB
    Color_Day=VectorParameter() {group=Daytime Smoke, parameterName=Daytime Mids Smoke, defaultValue=<rgba 0.3065 0.3412 0.445 0>}.RGB
    Color_Evening=VectorParameter() {group=Evening Smoke, parameterName=Evening Mids Smoke, defaultValue=<rgba 0.445 0.2898 0.2599 0>}.RGB
    Color_Night=VectorParameter() {group=Nighttime Smoke, parameterName=Nighttime Mids Smoke, defaultValue=<rgba 0.1206 0.1342 0.175 0>}.RGB
   ) {desc=Mids, materialFunction=&Color_Time_of_Day}.Result
   Alpha=Saturate(None=Add(A=@n10, B=@n11.Result))
  ) {desc=Dark / Mid Lerp}
  n13 = VectorParameter() {group=Nighttime Smoke, parameterName=Nighttime Highlights Smoke, defaultValue=<rgba 0.2489 0.3198 0.38 0>}
  n14 = MaterialFunctionCall(
   Color_Morning=VectorParameter() {group=Morning Smoke, parameterName=Morning Highlights Smoke, defaultValue=<rgba 0.555 0.4848 0.5064 0>}.RGB
   Color_Day=VectorParameter() {group=Daytime Smoke, parameterName=Daytime Highlights Smoke, defaultValue=<rgba 0.5172 0.5669 0.685 0>}.RGB
   Color_Evening=VectorParameter() {group=Evening Smoke, parameterName=Evening Highlights Smoke, defaultValue=<rgba 0.56 0.4479 0.4532 0>}.RGB
   Color_Night=@n13.RGB
  ) {desc=Highlights, materialFunction=&Color_Time_of_Day}
  n15 = ScalarParameter() {defaultValue=-2, group=Morning Smoke, parameterName=Cell Shaded Mid/Light Bias Morning Smoke}
  n16 = ScalarParameter() {defaultValue=-2, group=Daytime Smoke, parameterName=Cell Shaded Mid/Light Bias Day Smoke}
  n17 = ScalarParameter() {defaultValue=-2, group=Evening Smoke, parameterName=Cell Shaded Mid/Light Bias Evening Smoke}
  n18 = ScalarParameter() {defaultValue=-2, group=Nighttime Smoke, parameterName=Cell Shaded Mid/Light Bias Night Smoke}
  n19 = Saturate(None=Add(
    A=@n9
    B=MaterialFunctionCall(Color_Morning=@n15, Color_Day=@n16, Color_Evening=@n17, Color_Night=@n18) {materialFunction=&Color_Time_of_Day}.Result
   ))
  n20 = StaticSwitchParameter(True=Multiply(A=LinearInterpolate(A=@n12, B=@n14.Result, Alpha=@n19) {desc=Mid / Light Lerp}, B=ParticleColor().RGB), False=Constant()) {
   defaultValue=true
   desc="If disabled, this turns off smoke shading and will output black instead."
   group=Cell Shading Parameters
   parameterName=Use Smoke?
  }
  n21 = Multiply(A=@n20, B=ScalarParameter() {defaultValue=0.25, group=Cell Shading Parameters, parameterName=Smoke Emissive Boost}) {constB=0.5}
  n22 = ScalarParameter() {defaultValue=0.04, group=Explosion, parameterName=Main Shading Fresnel Brightness Explosion}
  n23 = DynamicParameter() {paramNames [Explode Smoke LERP, Explosion Brightness Fresnel, Panner Time Random, Morph Frame Animation]}
  n24 = VectorParameter() {group=Explosion, parameterName=Explosion Light Vector, defaultValue=<rgba 0.7356 0.4513 1 0>}
  n25 = CameraVectorWS()
  n26 = ScalarParameter() {defaultValue=0.9, group=Explosion, parameterName=Sun/Moon Directional Shading Blend Explosion}
  n27 = Multiply(
   A=MaterialFunctionCall(
    Brightness=Add(A=@n22, B=@n23.Explosion Brightness Fresnel)
    Exponent=ScalarParameter() {defaultValue=4, group=Explosion, parameterName=Main Shading Fresnel Exponent}
    Camera Vector=LinearInterpolate(A=@n24.RGB, B=@n25, Alpha=@n26)
   ) {materialFunction=&MF_Fresnel_FX}.Fresnel
   B=ScalarParameter() {defaultValue=3, group=Explosion, parameterName=Cell Shaded Step Day}
  )
  n28 = Ceil(None=Add(A=ScalarParameter() {defaultValue=-1, group=Explosion, parameterName=Cell Shaded Dark/Mid Bias Day}, B=@n27))
  n29 = Add(A=@n27, B=ScalarParameter() {defaultValue=-2.75, group=Explosion, parameterName=Cell Shaded Mid/Light Bias Day})
  n30 = LinearInterpolate(
   A=LinearInterpolate(
    A=VectorParameter() {group=Explosion, parameterName=Darks Explosion, defaultValue=<rgba 0.38 0.0229 0 0>}.RGB
    B=VectorParameter() {group=Explosion, parameterName=Mids Explosion, defaultValue=<rgba 5 1.9173 0.55 0>}.RGB
    Alpha=Saturate(None=@n28)
   ) {desc=Dark / Mid Lerp}
   B=VectorParameter() {group=Explosion, parameterName=Highlights Explosion, defaultValue=<rgba 6 4.3886 1.35 0>}.RGB
   Alpha=Saturate(None=Ceil(None=@n29))
  ) {desc=Mid / Light Lerp}
  n31 = MaterialFunctionCall() {materialFunction=&MF_Fresnel_FX}
  n32 = StaticSwitchParameter(
   True=ScalarParameter() {defaultValue=1, group=Explosion, parameterName=Explosion Debug LERP}
   False=ConstantBiasScale(None=DynamicParameter() {paramNames [Explode Smoke LERP, Explosion Brightness Fresnel, Panner Time Random, Morph Frame Animation]}.Explode Smoke LERP) {bias=-1, desc=Allows 0-1 Space in Cascade, scale=2}
  ) {desc="If disabled, this turns off smoke shading and will output black instead.", group=Explosion, parameterName=Debug Explosion LERP?}
  n33 = Ceil(None=Saturate(None=Add(A=@n31.Fresnel Inverted, B=@n32) {constB=0}))
  MaterialFunctionCall(Emissive=StaticSwitchParameter(True=LinearInterpolate(A=@n21, B=@n30, Alpha=@n33), False=@n21) {desc="If disabled, this turns off smoke shading and will output black instead.", group=Explosion, parameterName=Use Explosion?}) {materialFunction=&MF_FX_EmissiveAdjustments}
  n34 = StaticSwitchParameter(True=TextureCoordinate() {coordinateIndex=1}, False=TextureCoordinate()) {group=Normal Map, parameterName=Normal Use 2nd UV Set?}
  n35 = MaterialFunctionCall(Float2=@n34) {materialFunction=&BreakOutFloat2Components}
  n36 = Multiply(
   A=AppendVector(
    A=ScalarParameter() {defaultValue=1, group=Normal Map, parameterName=Normal Map U Tile}
    B=ScalarParameter() {defaultValue=1, group=Normal Map, parameterName=Normal Map V Tile}
   )
   B=StaticSwitchParameter(
    True=MaterialFunctionCall(X=@n35.G, Y=@n35.G) {materialFunction=&MakeFloat2}.Result
    False=@n34
   ) {group=Opacity, parameterName=Normal Flip UV}
  )
  n37 = Multiply(
   A=Time()
   B=DynamicParameter() {paramNames [Explode Smoke LERP, Explosion Brightness Fresnel, Panner Time Random, Morph Frame Animation]}.Panner Time Random
  )
  n38 = Multiply(A=@n37, B=ScalarParameter() {defaultValue=1, group=Normal Map, parameterName=Normal Map Panner U Speed})
  n39 = Multiply(A=@n37, B=ScalarParameter() {defaultValue=1, group=Normal Map, parameterName=Normal Map Panner V Speed})
  n40 = AppendVector(
   A=ComponentMask(None=Panner(Coordinate=@n36, Time=@n38) {bFractionalPart=true, speedX=0.5}) {r=true}
   B=ComponentMask(None=Panner(Coordinate=@n36, Time=@n39) {bFractionalPart=true, speedY=0.5}) {g=true}
  )
  n41 = StaticSwitchParameter(True=TextureCoordinate() {coordinateIndex=1}, False=TextureCoordinate()) {group=Opacity, parameterName=Opacity Mask Use 2nd UV Set?}
  n42 = MaterialFunctionCall(Float2=@n41) {materialFunction=&BreakOutFloat2Components}
  n43 = MaterialFunctionCall(X=@n42.G, Y=@n42.G) {materialFunction=&MakeFloat2}
  n44 = TextureSampleParameter2D(UVs=StaticSwitchParameter(True=@n40, False=StaticSwitchParameter(True=@n43.Result, False=@n41) {group=Opacity, parameterName=Opacity Mask Flip UV}) {group=Opacity, parameterName=Pan Opacity From Normal Map?}) {group=Opacity, parameterName=Opacity Texture}
  n45 = ScalarParameter() {
   defaultValue=3
   desc="Determines the \"roll\" to erode the mesh with. Keep this in a reasonable range (1 to 8)"
   group=Opacity
   parameterName=Erosion Fresnel Exponent
  }
  n46 = Add(
   A=MaterialFunctionCall(Exponent=@n45) {materialFunction=&MF_Fresnel_FX}.Fresnel
   B=ParticleColor().A
  )
  Multiply(A=Saturate(None=Multiply(A=@n44.A, B=@n46)), B=VertexColor().A)
  n47 = TextureSampleParameter2D(UVs=StaticSwitchParameter(True=@n40, False=@n36) {group=Normal Map, parameterName=Pan Normal Map?}) {
   group=Normal Map
   parameterName=Normal Map
   samplerType=SAMPLERTYPE_Normal
   texture=/Game/Effects/Fort_Effects/Gameplay/SmokeDustExplosions/Textures/T_Explosion_NM_01
  }
  Normalize(VectorInput=Multiply(A=@n47.RGB, B=VectorParameter() {group=Normal Map, parameterName=Normal Map Intensity, defaultValue=<rgba 1 1 1 0>}.RGB))
  ScalarParameter() {defaultValue=0.1, group=Cell Shading Parameters, parameterName=Metallic}
  ScalarParameter() {defaultValue=1, group=Cell Shading Parameters, parameterName=Roughness}
  ScalarParameter() {group=Cell Shading Parameters, parameterName=Specular}
 ]
}
